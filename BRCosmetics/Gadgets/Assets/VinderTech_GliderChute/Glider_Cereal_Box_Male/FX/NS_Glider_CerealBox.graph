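# Niagara trail and deploy effects for the Cereal Box glider, in rainbow-sparkle and fire styles.
# User.Variant indexes arrays that gate each emitter's spawning: 0 and 1 give rainbow ribbons and sparkles, 2 gives fire.
# Emitters combine skeletal-mesh sprite bursts, anim trails and thin ribbons that inherit source movement.
# User.ForwardDot scales ribbon backward velocity, and bIsFrontEnd adds vertical acceleration to trailing particles.
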NiagaraSystem /BRCosmetics/Gadgets/Assets/VinderTech_GliderChute/Glider_Cereal_Box_Male/FX/NS_Glider_CerealBox
dependencies {
 usedDataInterfaces [
  &NiagaraDataInterfaceArrayInt32
  &NiagaraDataInterfaceCurve
  &NiagaraDataInterfaceArrayFloat
  &NiagaraDataInterfaceArrayColor
  &NiagaraDataInterfaceColorCurve
  &NiagaraDataInterfaceArrayFloat3
 ]
 usedDynamicInputs []
 usedModules [&SystemState]
}
emitters {
 DEBUG_FORWARD_DOT {
  bEnabled=false
  emitterUpdateScript [
   &EmitterState()
   &SpawnRate(SpawnRate=10, Use Spawn Probability=false)
  ]
  particleSpawnScript [&InitializeParticle(
    Color=&ColorFromCurve(
     ColorCurve=DataInterface() {propertyValues {
       AlphaCurve {keys [<at 0 1 RCIM_Cubic>]}
       BlueCurve {keys [<at 0 0.4131 RCIM_Cubic>,<at 0.4587 1>,<at 1 0 RCIM_Cubic>]}
       GreenCurve {keys [<at 0 0 RCIM_Cubic>,<at 0.4587 0.2843>,<at 1 1 RCIM_Cubic>]}
       RedCurve {keys [<at 0 1 RCIM_Cubic>,<at 0.4587 0>,<at 1 0.1928 RCIM_Cubic>]}
      }}
     CurveIndex=&RemapRange(Input Value To Remap=$User.ForwardDot, Input Min=-1, Input Max=1, Output Min=0, Output Max=1)
    )
    Lifetime=5
    UsePositionOffset=false
    Write Material Random=true
   )]
  particleUpdateScript [&ParticleState()]
  rendererClasses [&NiagaraSpriteRendererProperties]
  renderers [{rendererClass=&NiagaraSpriteRendererProperties}]
 }
 Fire_Deploy001 {
  emitterUpdateScript [
   &EmitterState(Life Cycle Mode=::NewEnumerator1, Loop Behavior=::NewEnumerator1, Loop Duration=1, UseLoopDelay=false)
   &SpawnBurst_Instantaneous(
    Spawn Count=&SelectIntFromArray(
     Direct Array Index=&MakeIntFromFloat(Float=$User.Variant)
     Int Selection Array=DataInterface() {propertyValues {IntData [0,0,100]}}
    )
    Use Spawn Probability=false
    UseLoopCountLimit=false
   )
   &SpawnRate(
    SpawnRate=&SelectFloatFromArray(
     Direct Array Index=&MakeIntFromFloat(Float=$User.Variant)
     Float Selection Array=DataInterface() {propertyValues {FloatData [0,0,15]}}
    )
    Use Spawn Probability=false
   )
  ]
  particleSpawnScript [
   &InitializeParticle(
    Color Maximum=<rgba 1 0 0.0243 1>
    Color Minimum=<rgba 1 0.9697 0 1>
    Color Mode=::NewEnumerator2
    Lifetime=&RandomRangeFloat(Minimum=0.7, Maximum=1)
    Sprite Rotation Mode=::NewEnumerator2
    Sprite Size Mode=::NewEnumerator1
    Uniform Sprite Size Max=9
    Uniform Sprite Size Min=4.5
    UsePositionOffset=false
    Write Material Random=true
   )
   &SkeletalMeshLocation(
    MaskSpawnGroup=false
    OverrideRandomSeed=false
    Position Sampling=::NewEnumerator1
    Sampled Position Offset=&RandomRangeVector(Minimum=<-10 -10 -10>, Maximum=<10 10 10>)
    Skeletal Mesh=$User.SkeletalMesh
   )
   &AddVelocityFromPoint(Module.Use Falloff Distance=false, Velocity Strength=&RandomRangeFloat(Minimum=100, Maximum=250))
  ]
  particleUpdateScript [
   &ParticleState()
   &ScaleSpriteSize(Scale Factor=&Vector2DFromFloat(Value=&FloatFromCurve(FloatCurve=DataInterface() {propertyValues {Curve {keys [{interpMode=RCIM_Cubic, tangentMode=RCTM_User, time=0.25, value=1}, {interpMode=RCIM_Cubic, tangentMode=RCTM_User, time=1, value=0}]}}})))
   &InheritSourceMovement(
    Applied Force Scale=&VectorFromFloat(Value=&FloatFromCurve(FloatCurve=DataInterface() {propertyValues {Curve {keys [{interpMode=RCIM_Cubic, tangentMode=RCTM_User, time=0, value=1}, {interpMode=RCIM_Cubic, tangentMode=RCTM_User, time=1, value=0}]}}}))
    Applied Position Scale=&VectorFromFloat(Value=&FloatFromCurve(
      FloatCurve=DataInterface() {propertyValues {Curve {keys [
          {arriveTangent=-1.2152, interpMode=RCIM_Cubic, leaveTangent=-1.2152, tangentMode=RCTM_User, time=0, value=1}
          {interpMode=RCIM_Cubic, tangentMode=RCTM_User, time=1, value=0}
         ]}}}
      Scale Curve=0.95
     ))
    ApplyForce=true
    ApplyForceFalloff=false
    ApplyPosition=true
    ApplyPositionFalloff=false
    Module.LimitForceSpeed=false
   )
   &SolveForcesAndVelocity(Clamp Velocity=false, Limit Acceleration=false)
   &ScaleColor(
    Scale Alpha=&FloatFromCurve(FloatCurve=DataInterface() {propertyValues {Curve {keys [{interpMode=RCIM_Cubic, tangentMode=RCTM_User, time=0, value=1}, {interpMode=RCIM_Cubic, tangentMode=RCTM_User, time=1, value=0}]}}})
    ScaleA=true
    ScaleRGB=true
   )
  ]
  rendererClasses [&NiagaraSpriteRendererProperties]
  renderers [{rendererClass=&NiagaraSpriteRendererProperties}]
 }
 Fire_Ribbon {
  emitterUpdateScript [
   &EmitterState(Life Cycle Mode=::NewEnumerator1, Loop Duration=0.1, UseLoopDelay=false)
   &FN_SpawnAnimTrail_v2(
    Number of Anim Trails:NiagaraInt32=&SelectIntFromArray(
     Direct Array Index=&MakeIntFromFloat(Float=$User.Variant)
     Int Selection Array=DataInterface() {propertyValues {IntData [0,0,2]}}
    )
    Spawn Mode:ENiagaraSkeletalSpawning=::NewEnumerator0
   )
  ]
  particleSpawnScript [
   /Game/Effects/Niagara/Modules/AnimTrail/FN_InitializeAnimTrail/FN_InitializeAnimTrailBoneIndecies(Second Bone or Socket Index=1)
   &FN_InitializeAnimTrailBoneIndecies(AnimTrail Selection=1, First Bone or Socket Index=2, Second Bone or Socket Index=3) {module=FN_InitializeAnimTrailBoneIndecies001}
   /Game/Effects/Niagara/Modules/AnimTrail/FN_InitializeAnimTrail/FN_InitializeAnimTrail(
    AnimTrailWidth=3.5
    Position Offset:Vector3f=&RotateVectorByQuaternion(Vector To Rotate=<0 10 0>, Quaternion=$Engine.Owner.Rotation)
    Spawn Mode:ENiagaraSkeletalSpawning=::NewEnumerator0
    Write Material Random=false
   )
   &AddVelocity(Velocity=&MakeVector(
     X=0
     Y=&FloatFromCurve(
      FloatCurve=DataInterface() {propertyValues {Curve {keys [<at 0 0.25>,<at 0.5 1>]}}}
      CurveIndex=&RemapRange(Input Value To Remap=$User.ForwardDot, Input Min=-1, Input Max=1, Output Min=0, Output Max=1)
      Scale Curve=-500
     )
     Z=0
    ))
  ]
  particleUpdateScript [
   &InheritSourceMovement(
    Applied Position Scale=&Multiply_VectorByFloat(
     Vector=<0.7 0.7 1>
     Float=&FloatFromCurve(FloatCurve=DataInterface() {propertyValues {Curve {keys [{interpMode=RCIM_Cubic, tangentMode=RCTM_User, time=0, value=1}, {interpMode=RCIM_Cubic, tangentMode=RCTM_User, time=1, value=0.3}]}}})
    )
    ApplyForce=false
    ApplyPosition=true
    ApplyPositionFalloff=false
   )
   &AccelerationForce(
    Acceleration=&MakeVector(
     X=0
     Y=0
     Z=&MakeCustomFloatFromBool(
      Bool=$User.bIsFrontEnd
      True Float=&FloatFromCurve(FloatCurve=DataInterface() {propertyValues {Curve {keys [<at 0 -1>,<at 1 2.5>]}}}, Scale Curve=2000)
      False Float=0
     )
    )
    Coordinate Space=::Simulation
   )
   &SolveForcesAndVelocity(Clamp Velocity=false, Limit Acceleration=false)
   &FN_UpdateAnimTrail_v2()
   &ScaleRibbonWidth()
   &FN_RibbonMaterial(Color=&ColorFromCurve(ColorCurve=DataInterface() {propertyValues {
       AlphaCurve {keys [<at 0 0 RCIM_Cubic>,<at 0.3783 1>,<at 1 0 RCIM_Cubic>]}
       BlueCurve {keys [
         <at 0 0.7031>
         {arriveTangent=-3.2152, interpMode=RCIM_Cubic, leaveTangent=-3.2152, time=0.0982, value=0.0226}
         <at 0.2106 0.026>
         <at 0.7164 0 RCIM_Cubic>
        ]}
       GreenCurve {keys [
         <at 0 0.9965>
         {arriveTangent=-3.0312, interpMode=RCIM_Cubic, leaveTangent=-3.0312, time=0.0982, value=0.5605}
         <at 0.2106 0.3582>
         <at 0.7164 0.0379 RCIM_Cubic>
        ]}
       RedCurve {keys [<at 0 1>,<at 0.0982 1 RCIM_Cubic>,<at 0.2106 1>,<at 0.7164 1 RCIM_Cubic>]}
      }}))
   /SetVariables(Particles.Renderer2_RibbonWidth:NiagaraFloat=&Multiply_Float(A=$Particles.RibbonWidth, B=0.6)) {isSetParameters=true}
  ]
  rendererClasses [&NiagaraRibbonRendererProperties]
  renderers [
   {rendererClass=&NiagaraRibbonRendererProperties}
   {rendererClass=&NiagaraRibbonRendererProperties, rendererIndex=1}
  ]
 }
 Rainbo_Ribbon {
  emitterUpdateScript [
   &EmitterState(Life Cycle Mode=::NewEnumerator1, Loop Duration=0.1, UseLoopDelay=false)
   &FN_SpawnAnimTrail_v2(
    Number of Anim Trails:NiagaraInt32=&SelectIntFromArray(
     Direct Array Index=&MakeIntFromFloat(Float=$User.Variant)
     Int Selection Array=DataInterface() {propertyValues {IntData [2,2,0]}}
    )
    Spawn Mode:ENiagaraSkeletalSpawning=::NewEnumerator0
   )
  ]
  particleSpawnScript [
   &FN_InitializeAnimTrailBoneIndecies(Second Bone or Socket Index=1)
   &FN_InitializeAnimTrailBoneIndecies(AnimTrail Selection=1, First Bone or Socket Index=2, Second Bone or Socket Index=3) {module=FN_InitializeAnimTrailBoneIndecies001}
   &FN_InitializeAnimTrail(AnimTrailWidth=2, Spawn Mode:ENiagaraSkeletalSpawning=::NewEnumerator0, Write Material Random=false)
   &AddVelocity(Velocity=&MakeVector(
     X=0
     Y=&FloatFromCurve(
      FloatCurve=DataInterface() {propertyValues {Curve {keys [<at 0 0.25>,<at 0.5 1>]}}}
      CurveIndex=&RemapRange(Input Value To Remap=$User.ForwardDot, Input Min=-1, Input Max=1, Output Min=0, Output Max=1)
      Scale Curve=-500
     )
     Z=0
    ))
  ]
  particleUpdateScript [
   &InheritSourceMovement(
    Applied Position Scale=&Multiply_VectorByFloat(
     Vector=<0.7 0.7 1>
     Float=&FloatFromCurve(FloatCurve=DataInterface() {propertyValues {Curve {keys [{interpMode=RCIM_Cubic, tangentMode=RCTM_User, time=0, value=1}, {interpMode=RCIM_Cubic, tangentMode=RCTM_User, time=1, value=0.3}]}}})
    )
    ApplyForce=false
    ApplyPosition=true
    ApplyPositionFalloff=false
   )
   &AccelerationForce(
    Acceleration=&MakeVector(
     X=0
     Y=0
     Z=&MakeCustomFloatFromBool(
      Bool=$User.bIsFrontEnd
      True Float=&FloatFromCurve(FloatCurve=DataInterface() {propertyValues {Curve {keys [<at 0 -1>,<at 1 2.5>]}}}, Scale Curve=2000)
      False Float=0
     )
    )
    Coordinate Space=::Simulation
   )
   /SetVariables(Particles.Velocity=&Add_Vector(
     A=$Particles.Velocity
     B=&MakeCustomVectorFromBool(
      Bool=$User.bIsFrontEnd
      True Vector=&RotateVectorByQuaternion(
       Vector To Rotate=&MakeVector(
        X=&FloatFromCurve(FloatCurve=DataInterface() {propertyValues {Curve {keys [<at 0 -0.3>,<at 1 0.8>]}}}, Scale Curve=200)
        Y=0
        Z=0
       )
       Quaternion=$Engine.Owner.Rotation
      )
      False Vector=<1 0 0>
     )
    )) {enabled=false, isSetParameters=true}
   &SolveForcesAndVelocity(Clamp Velocity=false, Limit Acceleration=false)
   &FN_UpdateAnimTrail_v2()
   &ScaleRibbonWidth()
   &FN_RibbonMaterial() {enabled=false}
   &ScaleColor(
    Scale Alpha=&FloatFromCurve(
     FloatCurve=DataInterface() {propertyValues {Curve {keys [{interpMode=RCIM_Cubic, tangentMode=RCTM_User, time=0, value=1}, {interpMode=RCIM_Cubic, tangentMode=RCTM_User, time=1, value=0}]}}}
     Scale Curve=&FloatFromCurve(
      FloatCurve=DataInterface() {propertyValues {Curve {keys [
          {interpMode=RCIM_Cubic, tangentMode=RCTM_User, time=0, value=1}
          {interpMode=RCIM_Cubic, tangentMode=RCTM_User, time=0.5, value=0}
          {tangentMode=RCTM_User, time=1, value=1}
         ]}}}
      CurveIndex=$System.FN_Integrator.Value
     )
    )
    Scale RGB=&VectorFromFloat(Value=&FNTimeOfDayScalar(Night Scale=1, Morning Scale=1, Day Scale=2, Evening Scale=1))
    ScaleA=true
    ScaleRGB=true
   )
   /SetVariables(Particles.Renderer2_RibbonWidth:NiagaraFloat=&Multiply_Float(A=$Particles.RibbonWidth, B=0.9)) {isSetParameters=true}
   /SetVariables(Particles.Renderer2Position:Vector3f=&Add_Vector(A=Linked() {linkedVariable {name=Particles.Position, type=NiagaraPosition}}, B=<0 0 -5>)) {isSetParameters=true}
  ]
  rendererClasses [&NiagaraRibbonRendererProperties]
  renderers [
   {rendererClass=&NiagaraRibbonRendererProperties}
   {rendererClass=&NiagaraRibbonRendererProperties, rendererIndex=1}
  ]
 }
 Sparckle_Attached {
  emitterUpdateScript [
   &EmitterState()
   &SpawnRate(
    SpawnRate=&SelectFloatFromArray(
     Direct Array Index=&MakeIntFromFloat(Float=$User.Variant)
     Float Selection Array=DataInterface() {propertyValues {FloatData [15,15,0]}}
    )
    Use Spawn Probability=false
   )
  ]
  particleSpawnScript [&InitializeParticle(
    Color=&SelectLinearColorFromArray(
     Array Sampling Mode=::NewEnumerator1
     Direct Array Index=&MakeIntFromFloat(Float=$User.Variant)
     Color Selection Array=DataInterface() {propertyValues {ColorData [<rgba 313.368 410.2669 500 0.01>, <rgba 500 375.2358 92.8819 0.01>, <rgba 500 375.2358 92.8819 0.01>]}}
     Random Seed=?
    )
    Lifetime=&RandomRangeFloat(Minimum=0.7, Maximum=1)
    Sprite Rotation Angle=&RandomRangeFloat(Minimum=-0.02, Maximum=0.02)
    Sprite Rotation Mode=::NewEnumerator0
    Sprite Size Mode=::NewEnumerator1
    Uniform Sprite Size Max=9
    Uniform Sprite Size Min=4.5
    UsePositionOffset=false
    Write Material Random=true
   )]
  particleUpdateScript [
   &ParticleState()
   &SkeletalMeshLocation(MaskSpawnGroup=false, OverrideRandomSeed=false, Skeletal Mesh=$User.SkeletalMesh)
   &KillParticlesInVolume(Box Size:Vector3f=<75 100 300>, Invert Volume=true, Kill Shape=::NewEnumerator1, Origin Offset:Vector3f=<0 -25 0>)
   &KillParticlesInVolume(Origin Offset:Vector3f=<20 0 0>, Sphere Radius=25) {module=KillParticlesInVolume001}
   &ScaleSpriteSize(Scale Factor=&Vector2DFromFloat(Value=&FloatFromCurve(FloatCurve=DataInterface() {propertyValues {Curve {keys [{interpMode=RCIM_Cubic, tangentMode=RCTM_User, time=0.25, value=1}, {interpMode=RCIM_Cubic, tangentMode=RCTM_User, time=1, value=0}]}}})))
   &SolveForcesAndVelocity(Clamp Velocity=false, Limit Acceleration=false)
   &ScaleColor(
    Scale Alpha=&FloatFromCurve(FloatCurve=DataInterface() {propertyValues {Curve {keys [{interpMode=RCIM_Cubic, tangentMode=RCTM_User, time=0, value=1}, {interpMode=RCIM_Cubic, tangentMode=RCTM_User, time=1, value=0}]}}})
    ScaleA=true
    ScaleRGB=true
   )
  ]
  rendererClasses [&NiagaraSpriteRendererProperties]
  renderers [{rendererClass=&NiagaraSpriteRendererProperties}]
 }
 Sparckle_Deploy {
  emitterUpdateScript [
   &EmitterState(Life Cycle Mode=::NewEnumerator1, Loop Behavior=::NewEnumerator1, Loop Duration=1, UseLoopDelay=false)
   &SpawnBurst_Instantaneous(
    Spawn Count=&SelectIntFromArray(
     Direct Array Index=&MakeIntFromFloat(Float=$User.Variant)
     Int Selection Array=DataInterface() {propertyValues {IntData [100,100,0]}}
    )
    Use Spawn Probability=false
    UseLoopCountLimit=false
   )
   &SpawnRate(
    SpawnRate=&SelectFloatFromArray(
     Direct Array Index=&MakeIntFromFloat(Float=$User.Variant)
     Float Selection Array=DataInterface() {propertyValues {FloatData [15,15,0]}}
    )
    Use Spawn Probability=false
   )
  ]
  particleSpawnScript [
   &InitializeParticle(
    Color=<rgba 15 0 12.5 1>
    Lifetime=&RandomRangeFloat(Minimum=0.7, Maximum=1)
    Sprite Rotation Mode=::NewEnumerator2
    Sprite Size Mode=::NewEnumerator1
    Uniform Sprite Size Max=9
    Uniform Sprite Size Min=4.5
    UsePositionOffset=false
    Write Material Random=true
   )
   &SkeletalMeshLocation(
    MaskSpawnGroup=false
    OverrideRandomSeed=false
    Position Sampling=::NewEnumerator1
    Sampled Position Offset=&RandomRangeVector(Minimum=<-10 -10 -10>, Maximum=<10 10 10>)
    Skeletal Mesh=$User.SkeletalMesh
   )
   &AddVelocityFromPoint(Module.Use Falloff Distance=false, Velocity Strength=&RandomRangeFloat(Minimum=100, Maximum=250))
  ]
  particleUpdateScript [
   &ParticleState()
   &ScaleSpriteSize(Scale Factor=&Vector2DFromFloat(Value=&FloatFromCurve(FloatCurve=DataInterface() {propertyValues {Curve {keys [{interpMode=RCIM_Cubic, tangentMode=RCTM_User, time=0.25, value=1}, {interpMode=RCIM_Cubic, tangentMode=RCTM_User, time=1, value=0}]}}})))
   &InheritSourceMovement(
    Applied Force Scale=&VectorFromFloat(Value=&FloatFromCurve(FloatCurve=DataInterface() {propertyValues {Curve {keys [{interpMode=RCIM_Cubic, tangentMode=RCTM_User, time=0, value=1}, {interpMode=RCIM_Cubic, tangentMode=RCTM_User, time=1, value=0}]}}}))
    Applied Position Scale=&VectorFromFloat(Value=&FloatFromCurve(
      FloatCurve=DataInterface() {propertyValues {Curve {keys [
          {arriveTangent=-1.2152, interpMode=RCIM_Cubic, leaveTangent=-1.2152, tangentMode=RCTM_User, time=0, value=1}
          {interpMode=RCIM_Cubic, tangentMode=RCTM_User, time=1, value=0}
         ]}}}
      Scale Curve=0.95
     ))
    ApplyForce=true
    ApplyForceFalloff=false
    ApplyPosition=true
    ApplyPositionFalloff=false
    Module.LimitForceSpeed=false
   )
   &SolveForcesAndVelocity(Clamp Velocity=false, Limit Acceleration=false)
   &ScaleColor(
    Scale Alpha=&FloatFromCurve(FloatCurve=DataInterface() {propertyValues {Curve {keys [{interpMode=RCIM_Cubic, tangentMode=RCTM_User, time=0, value=1}, {interpMode=RCIM_Cubic, tangentMode=RCTM_User, time=1, value=0}]}}})
    ScaleA=true
    ScaleRGB=true
   )
  ]
  rendererClasses [&NiagaraSpriteRendererProperties]
  renderers [{rendererClass=&NiagaraSpriteRendererProperties}]
 }
 Sparckle_Flying {
  emitterUpdateScript [
   &EmitterState()
   &SpawnRate(
    SpawnRate=&SelectFloatFromArray(
     Direct Array Index=&MakeIntFromFloat(Float=$User.Variant)
     Float Selection Array=DataInterface() {propertyValues {FloatData [50,50,0]}}
    )
    Use Spawn Probability=false
   )
  ]
  particleSpawnScript [
   &InitializeParticle(
    Color=<rgba 15 0 12.5 1>
    Lifetime=&RandomRangeFloat(Minimum=0.7, Maximum=1)
    Sprite Rotation Mode=::NewEnumerator2
    Sprite Size Mode=::NewEnumerator1
    Uniform Sprite Size Max=6.5
    Uniform Sprite Size Min=4.5
    UsePositionOffset=false
    Write Material Random=true
   )
   &SkeletalMeshLocation(Bone Sampling Mode=::NewEnumerator0, MaskSpawnGroup=false, OverrideRandomSeed=false, Skeletal Mesh=$User.SkeletalMesh)
   &CurlNoiseForce(Noise Frequency=4, Noise Strength=75, Pan Noise=true, Pan Noise Field=<1 1 0>) {enabled=false}
   &ApplyInitialForces()
   &AddVelocity(Velocity=<0 -500 0>)
  ]
  particleUpdateScript [
   &ParticleState()
   &InheritSourceMovement(
    Applied Position Scale=&Multiply_VectorByFloat(
     Vector=<0.7 0.7 1>
     Float=&FloatFromCurve(FloatCurve=DataInterface() {propertyValues {Curve {keys [{interpMode=RCIM_Cubic, tangentMode=RCTM_User, time=0, value=1}, {interpMode=RCIM_Cubic, tangentMode=RCTM_User, time=1, value=0.3}]}}})
    )
    ApplyForce=false
    ApplyPosition=true
    ApplyPositionFalloff=false
   )
   &ScaleSpriteSize(Scale Factor=&Vector2DFromFloat(Value=&FloatFromCurve(FloatCurve=DataInterface() {propertyValues {Curve {keys [{interpMode=RCIM_Cubic, tangentMode=RCTM_User, time=0.25, value=1}, {interpMode=RCIM_Cubic, tangentMode=RCTM_User, time=1, value=0}]}}})))
   &AccelerationForce(
    Acceleration=&MakeVector(
     X=0
     Y=0
     Z=&MakeCustomFloatFromBool(
      Bool=$User.bIsFrontEnd
      True Float=&FloatFromCurve(FloatCurve=DataInterface() {propertyValues {Curve {keys [<at 0 -1>,<at 1 3>]}}}, Scale Curve=1500)
      False Float=0
     )
    )
    Coordinate Space=::Simulation
   )
   &Drag(Drag=1.5, Use Linear Drag=true, Use Rotational Drag=false)
   &SolveForcesAndVelocity(Clamp Velocity=false, Limit Acceleration=false)
   &ScaleColor(
    Scale Alpha=&FloatFromCurve(FloatCurve=DataInterface() {propertyValues {Curve {keys [{interpMode=RCIM_Cubic, tangentMode=RCTM_User, time=0, value=1}, {interpMode=RCIM_Cubic, tangentMode=RCTM_User, time=1, value=0}]}}})
    ScaleA=true
    ScaleRGB=true
   )
  ]
  rendererClasses [&NiagaraSpriteRendererProperties]
  renderers [{rendererClass=&NiagaraSpriteRendererProperties}]
 }
 ThinRibbons {
  emitterUpdateScript [
   &EmitterState()
   &FN_SpawnRibbon_v2(
    Number of Ribbons:NiagaraInt32=&SelectIntFromArray(
     Direct Array Index=&MakeIntFromFloat(Float=$User.Variant)
     Int Selection Array=DataInterface() {propertyValues {IntData [0,0,10]}}
    )
    Spawn Mode=::NewEnumerator0
    SpawnRate=30
   )
   &FN_Integrator() {enabled=false}
  ]
  particleSpawnScript [
   /Game/Effects/Niagara/Modules/Ribbons/FN_InitializeRibbon_v4(Enable Skeletal Position=true, Offset Amount=7, Ribbon Width=25, SkeletalMesh:NiagaraDataInterfaceSkeletalMesh=$User.Skeletal Mesh2)
   &FN_FlipRibbonV()
   &AddVelocity(Velocity=<0 0 100>) {enabled=false}
   &AddVelocity(Velocity=&MakeVector(
     X=0
     Y=&FloatFromCurve(
      FloatCurve=DataInterface() {propertyValues {Curve {keys [<at 0 0.25>,<at 0.5 1>]}}}
      CurveIndex=&RemapRange(Input Value To Remap=$User.ForwardDot, Input Min=-1, Input Max=1, Output Min=0, Output Max=1)
      Scale Curve=-500
     )
     Z=0
    )) {module=AddVelocity001}
  ]
  particleUpdateScript [
   &InheritSourceMovement(
    Applied Position Scale=&Multiply_VectorByFloat(
     Vector=<0.7 0.7 1>
     Float=&FloatFromCurve(FloatCurve=DataInterface() {propertyValues {Curve {keys [{interpMode=RCIM_Cubic, tangentMode=RCTM_User, time=0, value=1}, {interpMode=RCIM_Cubic, tangentMode=RCTM_User, time=1, value=0.3}]}}})
    )
    ApplyForce=false
    ApplyPosition=true
    ApplyPositionFalloff=false
   )
   &AccelerationForce(
    Acceleration=&MakeVector(
     X=0
     Y=0
     Z=&MakeCustomFloatFromBool(
      Bool=$User.bIsFrontEnd
      True Float=&FloatFromCurve(FloatCurve=DataInterface() {propertyValues {Curve {keys [<at 0 -1>,<at 1 3>]}}}, Scale Curve=2000)
      False Float=0
     )
    )
    Coordinate Space=::Simulation
   )
   &SolveForcesAndVelocity(Clamp Velocity=false, Limit Acceleration=false)
   /Game/Effects/Niagara/Modules/Ribbons/FN_UpdateRibbon_v4(Enable Skeletal Position=true, SkeletalMesh:NiagaraDataInterfaceSkeletalMesh=$User.Skeletal Mesh2)
   &FN_RibbonMaterial(
    Color=&ColorFromCurve(ColorCurve=DataInterface() {propertyValues {
       AlphaCurve {keys [<at 0.0112 0>,<at 0.2395 1>,<at 1 0 RCIM_Cubic>]}
       BlueCurve {keys [{arriveTangent=-1.6322, interpMode=RCIM_Cubic, time=0.019, value=0.1892}, <at 0.4148 0.026>, <at 1 0 RCIM_Cubic>]}
       GreenCurve {keys [{arriveTangent=-1.5388, interpMode=RCIM_Cubic, time=0.019, value=0.9445}, <at 0.4148 0.3582>, <at 1 0.0379 RCIM_Cubic>]}
       RedCurve {keys [<at 0.019 1 RCIM_Cubic>,<at 0.4148 1>,<at 1 1 RCIM_Cubic>]}
      }})
    Opacity Erosion:NiagaraFloat=&FloatFromCurve(FloatCurve=DataInterface() {propertyValues {Curve {keys [
         {interpMode=RCIM_Cubic, tangentMode=RCTM_User, time=0, value=0}
         {arriveTangent=2.6226, interpMode=RCIM_Cubic, leaveTangent=2.6226, tangentMode=RCTM_User, time=1, value=1}
        ]}}})
    Opacity Pan U Offset:NiagaraFloat=$Particles.MaterialRandom
    Opacity Pan U Speed=1.5
    Opacity Scale U=1.25
   )
   &ScaleRibbonWidth(Ribbon Width Scale=&Multiply_Float(
     A=&FloatFromCurve(FloatCurve=DataInterface() {propertyValues {Curve {keys [
          {arriveTangent=12.569, interpMode=RCIM_Cubic, leaveTangent=12.569, tangentMode=RCTM_User, time=0, value=0}
          <at 0.1 0.95>
          {interpMode=RCIM_Cubic, tangentMode=RCTM_User, time=1, value=0}
         ]}}})
     B=&RemapRange(Input Value To Remap=$Particles.MaterialRandom, Input Min=0, Input Max=1, Output Min=0.8, Output Max=1)
    ))
  ]
  rendererClasses [&NiagaraRibbonRendererProperties]
  renderers [{rendererClass=&NiagaraRibbonRendererProperties}]
 }
 ThinRibbonsOptimized {
  emitterUpdateScript [
   &EmitterState()
   &FN_SpawnRibbon_v2(
    Number of Ribbons:NiagaraInt32=&SelectIntFromArray(
     Direct Array Index=&MakeIntFromFloat(Float=$User.Variant)
     Int Selection Array=DataInterface() {propertyValues {IntData [10,10,0]}}
    )
    Spawn Mode=::NewEnumerator0
    SpawnRate=30
   )
   &FN_Integrator() {enabled=false}
  ]
  particleSpawnScript [
   &FN_InitializeRibbon_v4(Enable Skeletal Position=true, Offset Amount=5, Ribbon Width=25, SkeletalMesh:NiagaraDataInterfaceSkeletalMesh=$User.Skeletal Mesh2)
   &FN_FlipRibbonV()
   &AddVelocity(Velocity=<0 0 100>) {enabled=false}
   &AddVelocity(Velocity=&MakeVector(
     X=0
     Y=&FloatFromCurve(
      FloatCurve=DataInterface() {propertyValues {Curve {keys [<at 0 0.25>,<at 0.5 1>]}}}
      CurveIndex=&RemapRange(Input Value To Remap=$User.ForwardDot, Input Min=-1, Input Max=1, Output Min=0, Output Max=1)
      Scale Curve=-500
     )
     Z=0
    )) {module=AddVelocity001}
   /SetVariables(Particles.Position=&Add_Vector(
     A=Linked() {linkedVariable {name=Particles.Position, type=NiagaraPosition}}
     B=&SelectVectorFromArray(
      Direct Array Index=&ReturnParticleIDAttributes(Particle ID Attribute=::NewEnumerator1, Niagara ID=$Particles.RibbonID)
      Vector Selection Array=DataInterface() {propertyValues {FloatData [<0 -100 0>, <0 -100 0>, <0 -100 0>, <0 -100 0>, <0 -100 0>, <0 0 0>, <0 0 0>, <0 0 0>, <0 0 0>, <0 0 0>]}}
     )
    )) {enabled=false, isSetParameters=true}
  ]
  particleUpdateScript [
   &InheritSourceMovement(
    Applied Position Scale=&Multiply_VectorByFloat(
     Vector=<0.7 0.7 1>
     Float=&FloatFromCurve(FloatCurve=DataInterface() {propertyValues {Curve {keys [{interpMode=RCIM_Cubic, tangentMode=RCTM_User, time=0, value=1}, {interpMode=RCIM_Cubic, tangentMode=RCTM_User, time=1, value=0.3}]}}})
    )
    ApplyForce=false
    ApplyPosition=true
    ApplyPositionFalloff=false
   )
   &AccelerationForce(
    Acceleration=&MakeVector(
     X=0
     Y=0
     Z=&MakeCustomFloatFromBool(
      Bool=$User.bIsFrontEnd
      True Float=&FloatFromCurve(FloatCurve=DataInterface() {propertyValues {Curve {keys [<at 0 -1>,<at 1 3>]}}}, Scale Curve=2200)
      False Float=0
     )
    )
    Coordinate Space=::Simulation
   )
   /SetVariables(Particles.Velocity=&Add_Vector(
     A=$Particles.Velocity
     B=&MakeCustomVectorFromBool(
      Bool=$User.bIsFrontEnd
      True Vector=&RotateVectorByQuaternion(
       Vector To Rotate=&MakeVector(
        X=&FloatFromCurve(FloatCurve=DataInterface() {propertyValues {Curve {keys [<at 0 -0.3>,<at 1 0.8>]}}}, Scale Curve=200)
        Y=0
        Z=0
       )
       Quaternion=$Engine.Owner.Rotation
      )
      False Vector=<1 0 0>
     )
    )) {enabled=false, isSetParameters=true}
   &SolveForcesAndVelocity(Clamp Velocity=false, Limit Acceleration=false)
   &FN_UpdateRibbon_v4(Enable Skeletal Position=true, SkeletalMesh:NiagaraDataInterfaceSkeletalMesh=$User.Skeletal Mesh2)
   &FN_RibbonMaterial(
    Opacity Erosion:NiagaraFloat=&FloatFromCurve(FloatCurve=DataInterface() {propertyValues {Curve {keys [
         {interpMode=RCIM_Cubic, tangentMode=RCTM_User, time=0, value=0}
         {arriveTangent=2.6226, interpMode=RCIM_Cubic, leaveTangent=2.6226, tangentMode=RCTM_User, time=1, value=1}
        ]}}})
    Opacity Pan U Offset:NiagaraFloat=$Particles.MaterialRandom
    Opacity Pan U Speed=1.5
    Opacity Scale U=1.25
   )
   &ScaleRibbonWidth(Ribbon Width Scale=&Multiply_Float(
     A=&FloatFromCurve(FloatCurve=DataInterface() {propertyValues {Curve {keys [
          {arriveTangent=12.569, interpMode=RCIM_Cubic, leaveTangent=12.569, tangentMode=RCTM_User, time=0, value=0}
          <at 0.1 0.95>
          {interpMode=RCIM_Cubic, tangentMode=RCTM_User, time=1, value=0}
         ]}}})
     B=&RemapRange(Input Value To Remap=$Particles.MaterialRandom, Input Min=0, Input Max=1, Output Min=0.9, Output Max=1.2)
    ))
   /SetVariables(Particles.Color=&SelectLinearColorFromArray(
     Array Sampling Mode=::NewEnumerator1
     Direct Array Index=&ReturnParticleIDAttributes(Particle ID Attribute=::NewEnumerator1, Niagara ID=$Particles.RibbonID)
     Color Selection Array=DataInterface() {propertyValues {ColorData [
        <rgba 1 0 0.003 1>
        <rgba 1 0.0893 0 1>
        <rgba 1 0.2137 0 1>
        <rgba 0.1658 1 0 1>
        <rgba 0 1 0.7493 1>
        <rgba 0 1 0.6372 1>
        <rgba 0 0.373 1 1>
        <rgba 0 1 0.6393 1>
        <rgba 0 0.2813 1 1>
        <rgba 0.1218 0 1 1>
        <rgba 1 0 0.0334 1>
        <rgba 1 0.1961 0 1>
        <rgba 1 0.572 0 1>
        <rgba 0.1979 1 0 1>
        <rgba 0 1 0.5545 1>
        <rgba 0 1 0.2446 1>
        <rgba 0 0.7521 1 1>
        <rgba 0 1 0.618 1>
        <rgba 0 1 0.8868 1>
        <rgba 0.2717 0 1 1>
       ]}}
     Random Seed=?
    )) {enabled=false, isSetParameters=true}
   &ScaleColor(
    Scale Alpha=&FloatFromCurve(FloatCurve=DataInterface() {propertyValues {Curve {keys [
         {arriveTangent=1.0365, interpMode=RCIM_Cubic, leaveTangent=1.0365, tangentMode=RCTM_User, time=0, value=0}
         {interpMode=RCIM_Cubic, tangentMode=RCTM_User, time=0.3, value=0.95}
         {tangentMode=RCTM_User, time=1, value=0}
        ]}}})
    Scale RGB=&MakeVectorFromLinearColorRGB(Color=&SelectLinearColorFromArray(
      Array Sampling Mode=::NewEnumerator1
      Direct Array Index=&ReturnParticleIDAttributes(Particle ID Attribute=::NewEnumerator1, Niagara ID=$Particles.RibbonID)
      Color Selection Array=DataInterface() {propertyValues {ColorData [
         <rgba 1 0 0.003 1>
         <rgba 1 0.2137 0 1>
         <rgba 0.1658 1 0 1>
         <rgba 0 0.2813 1 1>
         <rgba 0.1218 0 1 1>
         <rgba 1 0 0.0334 1>
         <rgba 1 0.572 0 1>
         <rgba 0.1979 1 0 1>
         <rgba 0 0.2813 1 1>
         <rgba 0.2717 0 1 1>
        ]}}
      Random Seed=?
     ))
    ScaleA=true
    ScaleRGB=true
   )
  ]
  rendererClasses [&NiagaraRibbonRendererProperties]
  renderers [{rendererClass=&NiagaraRibbonRendererProperties}]
 }
}
summary {
 systemName=NS_Glider_CerealBox
 emitters [
  {emitterName=Rainbo_Ribbon, rendererClasses [&NiagaraRibbonRendererProperties]}
  {emitterName=Sparckle_Flying, rendererClasses [&NiagaraSpriteRendererProperties]}
  {emitterName=Sparckle_Attached, rendererClasses [&NiagaraSpriteRendererProperties]}
  {emitterName=Fire_Ribbon, rendererClasses [&NiagaraRibbonRendererProperties]}
  {emitterName=ThinRibbons, rendererClasses [&NiagaraRibbonRendererProperties]}
  {emitterName=Sparckle_Deploy, rendererClasses [&NiagaraSpriteRendererProperties]}
  {emitterName=Fire_Deploy001, rendererClasses [&NiagaraSpriteRendererProperties]}
  {bEnabled=false, emitterName=DEBUG_FORWARD_DOT, rendererClasses [&NiagaraSpriteRendererProperties]}
  {emitterName=ThinRibbonsOptimized, rendererClasses [&NiagaraRibbonRendererProperties]}
 ]
 userVariables [
  User.bIsFrontEnd {type=NiagaraBool}
  User.bIsFrontEndPreview {type=NiagaraBool}
  User.bisFullyDeployed {type=NiagaraBool}
  User.ForwardDot {type=NiagaraFloat, defaultValue=0, description=""}
  User.RibbonIDOffset {type=NiagaraInt32}
  User.Skeletal Mesh2 {
   type=NiagaraDataInterfaceSkeletalMesh
   defaultValue {
    struct=NiagaraExt_VariableValue_DataInterface
    value {dataInterface=&self:NiagaraDataInterfaceSkeletalMesh_0}
   }
  }
  User.SkeletalMesh {
   type=NiagaraDataInterfaceSkeletalMesh
   defaultValue {
    struct=NiagaraExt_VariableValue_DataInterface
    value {dataInterface=&self:NiagaraDataInterfaceSkeletalMesh_1}
   }
  }
  User.Variant {type=NiagaraFloat, defaultValue=0, description=""}
 ]
}
