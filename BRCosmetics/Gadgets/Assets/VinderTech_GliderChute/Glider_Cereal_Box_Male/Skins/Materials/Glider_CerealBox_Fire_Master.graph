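# Master material for the Cereal Box glider's fire skins, built on the character PreFX and PostFX functions with optional gem and sequin effects.
# PanningEmissive runs on rotated, X-mirrored pre-skinned positions, and a scrolling fire texture projected from local position modulates it inside a mask channel.
# An inverted Fresnel times a sphere mask blends that glow into the base emissive, then time of day scales the result.
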Material /BRCosmetics/Gadgets/Assets/VinderTech_GliderChute/Glider_Cereal_Box_Male/Skins/Materials/Glider_CerealBox_Fire_Master
graph {
 properties {
  bUseMaterialAttributes=true
  bUsedWithMeshDeformer=true
  bUsedWithSkeletalMesh=true
  pixelDepthOffsetMode=PDOM_Legacy
  shadingModel=MSM_FromMaterialExpression
  subsurfaceProfile=&SS_HeroSkin_02
  usedShadingModels=Default Lit | Subsurface Profile
 }
 nodes [
  n1 = MaterialFunctionCall() {materialFunction=&BaseFN_CharacterFunction_PreFX}
  n2 = TextureObjectParameter() {group=Skin FX, parameterName=SkinFX_Mask, samplerType=SAMPLERTYPE_Masks, texture=&T_WhiteMask}
  n3 = TextureSample(Tex=@n2) {samplerType=SAMPLERTYPE_Masks}
  n4 = StaticSwitchParameter(
   True=MaterialFunctionCall!(
    MASK=StaticComponentMaskParameter(None=@n3.RGBA) {defaultR=true, group=Skin FX, parameterName=Generic Gem Effect Channel}
    Color Change Sequence=StaticComponentMaskParameter(None=@n3.RGBA) {defaultB=true, group=Skin FX, parameterName=Generic Gem Color Change Channel}
    Material=@n1.Result
   ) {materialFunction=/Game/Weapons/WeaponSkins/MaterialFunction/GenericGem}.Result
   False=@n1.Result
  ) {group=Skin FX, parameterName=Use Generic Gem}
  n5 = GetMaterialAttributes(None=StaticSwitchParameter(
    True=MaterialFunctionCall!(Mask=@n2, TechArtMask, inMat=@n4) {materialFunction=/Game/Weapons/WeaponSkins/MaterialFunction/MF_Sequins}.Result
    False=@n4
   ) {group=Skin FX, parameterName=UseSequins}) {attributeGetTypes [#1]}
  n6 = StaticComponentMaskParameter(None=@n3.RGBA) {defaultR=true, group=Skin FX, parameterName=PanningEmissive_Channel}
  n7 = VertexInterpolator(VS=PreSkinnedPosition())
  n8 = Constant3Vector() {constant=<rgba -2 1 0 0>}
  n9 = AppendVector(A=Sign(None=ComponentMask(None=@n7.PS) {r=true}), B=Constant2Vector() {g=1, r=1})
  n10 = MaterialFunctionCall!(VectorToRotate=@n7.PS, LookAtVector=Normalize(VectorInput=Multiply(A=@n8, B=@n9)), RestingVector=Constant3Vector() {constant=<rgba 0 1 0 0>}) {materialFunction=/Engine/Functions/Engine_MaterialFunctions01/Vectors/RotateVector}
  n11 = Constant3Vector() {constant=<rgba 1 0.7 1 0>}
  n12 = GetMaterialAttributes(None=MaterialFunctionCall!(
    Alpha=@n6
    isVehicle
    Gmap
    UVOffset=StaticComponentMaskParameter(None=@n3.RGBA) {defaultB=true, group=Skin FX, parameterName=PanningEmissive_UVOffset_Channel}
    Local Normal
    Local Position=Multiply(A=@n10.Result, B=@n11)
    EmissiveColorOverride
    Material=@n1.Result
   ) {materialFunction=/Game/Weapons/WeaponSkins/MaterialFunction/PanningEmissive}.Result) {attributeGetTypes [#1]}
  n13 = LinearInterpolate(
   A=ComponentMask(None=@n1.Pre skinned Local Position) {b=true, r=true}
   B=ComponentMask(None=@n1.Pre skinned Local Position) {g=true, r=true}
   Alpha=ScalarParameter() {group=CerealBoxFire, parameterName=CB_FireTopDownUV}
  )
  n14 = ScalarParameter() {defaultValue=0.01, group=CerealBoxFire, parameterName=CB_FireTiling}
  n15 = Time()
  n16 = AppendVector(
   A=ScalarParameter() {defaultValue=1, group=CerealBoxFire, parameterName=CB_FireSpeedX}
   B=ScalarParameter() {defaultValue=1, group=CerealBoxFire, parameterName=CB_FireSpeedZ}
  )
  n17 = TextureSampleParameter2D(UVs=Add(A=Multiply(A=@n13, B=@n14), B=Multiply(A=@n15, B=@n16))) {
   group=CerealBoxFire
   parameterName=CB_FireTex
   samplerType=SAMPLERTYPE_LinearGrayscale
   texture=/BRCosmetics/Weapons/WeaponSkins/Wraps/Textures/GlowRandom/T_Wrap_GlowRandom_M
  }
  n18 = LinearInterpolate(
   A=ScalarParameter() {group=CerealBoxFire, parameterName=CB_FireMin}
   B=ScalarParameter() {defaultValue=10, group=CerealBoxFire, parameterName=CB_FireMax}
   Alpha=@n17.R
  )
  n19 = PixelNormalWS()
  n20 = StaticBool() {value=true}
  n21 = Constant() {r=2}
  n22 = SphereMask(A=PreSkinnedPosition(), B=Constant3Vector() {constant=<rgba 0 -100 0 0>}, Radius=Constant() {r=150}, Hardness=Constant() {r=0.7})
  n23 = LinearInterpolate(
   A=@n5.Emissive Color
   B=LinearInterpolate(A=@n12.Emissive Color, B=Multiply(A=@n12.Emissive Color, B=@n18), Alpha=@n6)
   Alpha=Multiply(
    A=MaterialFunctionCall(Normal Vector=@n19, Invert Fresnel=@n20, Power=@n21) {materialFunction=&Fresnel_Function}.Result
    B=VertexInterpolator(VS=@n22).PS
   )
  )
  n24 = MaterialFunctionCall(Value_Morning=Constant() {r=0.15}, Value_Day=Constant() {r=1}, Value_Evening=Constant() {r=0.25}, Value_Night=Constant() {r=0.8}) {materialFunction=&Scalar_Time_of_Day}
  n25 = MaterialFunctionCall(
   AO=@n1.AO
   Distance Blend=@n1.AO
   Fresnel=@n1.AO
   Gmap=@n1.AO
   Pre Skinned Local Position=@n1.AO
   Material=SetMaterialAttributes(MaterialAttributes=@n5.MaterialAttributes, Emissive Color=Multiply(A=@n23, B=@n24.Result)) {attributeSetTypes [#1]}
  ) {materialFunction=&BaseFN_CharacterFunction_PostFX}
  MaterialFunctionCall(In=@n25.Result) {materialFunction=&CINE_MF_CharPostFX}
 ]
}
